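# Sound cue for dropping the TeaCake dynamite consumable, played quietly with weapon-drop attenuation and the loot-collect sound class.
# A mixer layers a frag grenade drop wave, given slight random pitch variation, with a randomly chosen TNT ignite wave pitched up to double.
# It references the oldest-pickup-maximum concurrency set and carries concurrency override values for a single instance.
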SoundCue /TeaCakeConsumables/Assets/Audio/SC_TeaCake_Dynamite_Drop
root=Mixer_1
nodes [
 {id=Mixer_1, children [Modulator_1,Modulator_2], properties {inputVolume [1,1]}}
 {id=Modulator_1, children [WavePlayer_5], properties {pitchMax=1.1, pitchMin=0.9, volumeMax=1, volumeMin=1}}
 {id=Modulator_2, children [Random_1], properties {pitchMax=2, pitchMin=2}}
 {id=Random_1, children [WavePlayer_3,WavePlayer_2], properties {weights [1,1]}}
 {id=WavePlayer_2, properties {soundWaveAssetPtr=&/TNT_Ignite_04}}
 {id=WavePlayer_3, properties {soundWaveAssetPtr=&/TNT_Ignite_03}}
 {id=WavePlayer_5, properties {soundWaveAssetPtr=/Game/Sounds/Fort_Weapons/Grenades/Frag/Grenade_Frag_Drop_01}}
]
properties {
 attenuationSettings=&Fort_Weapon_Drop_Att
 maxDistance=3000
 soundClassObject=&Fort_Loot_Collect
 volumeMultiplier=0.2
 concurrencyOverrides {maxCount=1}
 concurrencySet [&PickupMaximum_Oldest]
}
