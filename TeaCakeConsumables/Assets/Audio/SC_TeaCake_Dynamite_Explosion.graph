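# Sound cue for the TeaCake dynamite explosion, layering close, medium and distant TNT blasts by distance.
# A mixer blends the three branches at descending volume, each with its own sound class, attenuation, pitch modulation and random wave choice.
# The distant pool also reuses a medium wave through an extra upward pitch modulator.
# A concurrency override caps simultaneous voices, stops the oldest and scales down volume.
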SoundCue /TeaCakeConsumables/Assets/Audio/SC_TeaCake_Dynamite_Explosion
root=Mixer_4
nodes [
 {
  id=Attenuation_0
  children [Modulator_1]
  properties {
   attenuationSettings=/Game/Sounds/Attenuation/Fort_Explosion_Distant_Athena_Att
   attenuationOverrides {falloffDistance=3000, attenuationShapeExtents {x=1000}}
  }
 }
 {
  id=Attenuation_4
  children [Modulator_0]
  properties {attenuationSettings=/Game/Sounds/Attenuation/Fort_Explosion_Close_Athena_Att, attenuationOverrides {falloffDistance=1600}}
 }
 {
  id=Attenuation_5
  children [Modulator_5]
  properties {
   attenuationSettings=/Game/Sounds/Attenuation/Fort_Explosion_Medium_Athena_Att
   attenuationOverrides {falloffDistance=3000, attenuationShapeExtents {x=1000}}
  }
 }
 {id=Mixer_4, children [SoundClass_3,SoundClass_2,SoundClass_0], properties {inputVolume [1,0.9,0.5]}}
 {id=Modulator_0, children [Random_0], properties {pitchMax=1.15, volumeMax=1, volumeMin=1}}
 {id=Modulator_1, children [Random_2], properties {pitchMax=1.2, pitchMin=0.8}}
 {id=Modulator_2, children [WavePlayer_0], properties {pitchMax=1.2, pitchMin=1}}
 {id=Modulator_5, children [Random_1], properties {pitchMax=1.2, pitchMin=0.85}}
 {
  id=Random_0
  children [WavePlayer_3,WavePlayer_5,WavePlayer_6,WavePlayer_7]
  properties {bShouldExcludeFromBranchCulling=true, weights [1,1,1,1]}
 }
 {id=Random_1, children [WavePlayer_10, WavePlayer_11, WavePlayer_12], properties {bShouldExcludeFromBranchCulling=true, weights [1,1,1]}}
 {id=Random_2, children [WavePlayer_13, WavePlayer_14, Modulator_2], properties {bShouldExcludeFromBranchCulling=true, weights [1,1,1]}}
 {
  id=SoundClass_0
  children [Attenuation_0]
  properties {soundClassOverride=&Fort_WeaponRanged_Fire_Distant}
 }
 {id=SoundClass_2, children [Attenuation_5], properties {soundClassOverride=/Game/Sounds/FortSoundClassesAndModes/Fort_Explosives}}
 {id=SoundClass_3, children [Attenuation_4], properties {soundClassOverride=&Fort_Explosives}}
 {id=WavePlayer_0, properties {soundWaveAssetPtr=/Game/Athena/Sounds/Weapons/Explosions/TNT_Expl_Med_03}}
 {id=WavePlayer_10, properties {soundWaveAssetPtr=/Game/Athena/Sounds/Weapons/Explosions/TNT_Expl_Med_01}}
 {id=WavePlayer_11, properties {soundWaveAssetPtr=/Game/Athena/Sounds/Weapons/Explosions/TNT_Expl_Med_02}}
 {id=WavePlayer_12, properties {soundWaveAssetPtr=&TNT_Expl_Med_03}}
 {id=WavePlayer_13, properties {soundWaveAssetPtr=/Game/Athena/Sounds/Weapons/Explosions/TNT_Expl_Dist_02}}
 {id=WavePlayer_14, properties {soundWaveAssetPtr=/Game/Athena/Sounds/Weapons/Explosions/TNT_Expl_Dist_01}}
 {id=WavePlayer_3, properties {soundWaveAssetPtr=/Game/Athena/Sounds/Weapons/Explosions/TNT_Explosion_Close_New_01}}
 {id=WavePlayer_5, properties {soundWaveAssetPtr=/Game/Athena/Sounds/Weapons/Explosions/TNT_Explosion_Close_New_02}}
 {id=WavePlayer_6, properties {soundWaveAssetPtr=/Game/Athena/Sounds/Weapons/Explosions/TNT_Explosion_Close_New_03}}
 {id=WavePlayer_7, properties {soundWaveAssetPtr=/Game/Athena/Sounds/Weapons/Explosions/TNT_Explosion_Close_New_04}}
]
properties {
 bOverrideConcurrency=true
 maxDistance=27500
 volumeMultiplier=0.8
 concurrencyOverrides {maxCount=5, resolutionRule=StopOldest, volumeScale=0.5}
}
