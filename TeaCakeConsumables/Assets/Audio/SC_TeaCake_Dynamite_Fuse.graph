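# Sound cue for a thrown TeaCake dynamite, layering the throw sound, fuse ignition and a looping fuse burn.
# A mixer blends a looping fuse that an enveloper fades in, random throw one-shots, and slightly quieter random fuse-start one-shots.
# Each layer has pitch or volume modulation and its own attenuation node, with ambient settings for the loop and throw settings for both one-shots.
# The cue uses the Explosives sound class and caps concurrency.
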SoundCue /TeaCakeConsumables/Assets/Audio/SC_TeaCake_Dynamite_Fuse
root=Mixer_0
nodes [
 {id=Attenuation_0, children [Modulator_1], properties {attenuationSettings=/Game/Sounds/Attenuation/Fort_Explosive_Fuse_Ambient_Att}}
 {id=Attenuation_2, children [Modulator_2], properties {attenuationSettings=/Game/Sounds/Attenuation/Fort_Explosive_Fuse_Prj_Throw_Att}}
 {id=Attenuation_4, children [Modulator_0], properties {attenuationSettings=&Fort_Explosive_Fuse_Prj_Throw_Att}}
 {
  id=Enveloper_0
  children [WavePlayer_1]
  properties {
   pitchCurve {editorCurveData {keys [{time=0}]}}
   volumeCurve {editorCurveData {keys [{arriveTangent=1.7767, interpMode=RCIM_Cubic, leaveTangent=1.7767, tangentMode=RCTM_User, time=0, value=0}, <at 1 1>]}}
  }
 }
 {id=Mixer_0, children [Attenuation_0, Attenuation_4, Attenuation_2], properties {inputVolume [1,1,0.8]}}
 {id=Modulator_0, children [Random_1], properties {pitchMax=1.2, volumeMax=1, volumeMin=0.8}}
 {id=Modulator_1, children [Enveloper_0], properties {pitchMax=1.2, pitchMin=0.9, volumeMax=1, volumeMin=1}}
 {id=Modulator_2, children [Random_0], properties {pitchMax=1.3}}
 {id=Random_0, children [WavePlayer_2,WavePlayer_3,WavePlayer_4], properties {bShouldExcludeFromBranchCulling=true, weights [1,1,1]}}
 {
  id=Random_1
  children [WavePlayer_18, WavePlayer_19, WavePlayer_20, WavePlayer_21]
  properties {bShouldExcludeFromBranchCulling=true, weights [1,1,1,1]}
 }
 {id=WavePlayer_1, properties {bLooping=true, soundWaveAssetPtr=/Game/Sounds/Fort_GamePlay_Sounds/Explosives/TnT/TNT_Fuse_Loop01}}
 {id=WavePlayer_18, properties {soundWaveAssetPtr=/Game/Sounds/Fort_GamePlay_Sounds/Explosives/TnT/TNT_Prj_Throw_01}}
 {id=WavePlayer_19, properties {soundWaveAssetPtr=/Game/Sounds/Fort_GamePlay_Sounds/Explosives/TnT/TNT_Prj_Throw_02}}
 {id=WavePlayer_2, properties {soundWaveAssetPtr=/Game/Sounds/Fort_GamePlay_Sounds/Explosives/TnT/TNT_Fuse_Prj_Loop_Start_04}}
 {id=WavePlayer_20, properties {soundWaveAssetPtr=/Game/Sounds/Fort_GamePlay_Sounds/Explosives/TnT/TNT_Prj_Throw_03}}
 {id=WavePlayer_21, properties {soundWaveAssetPtr=/Game/Sounds/Fort_GamePlay_Sounds/Explosives/TnT/TNT_Prj_Throw_04}}
 {id=WavePlayer_3, properties {soundWaveAssetPtr=/Game/Sounds/Fort_GamePlay_Sounds/Explosives/TnT/TNT_Fuse_Prj_Loop_Start_03}}
 {id=WavePlayer_4, properties {soundWaveAssetPtr=/Game/Sounds/Fort_GamePlay_Sounds/Explosives/TnT/TNT_Fuse_Prj_Loop_Start_02}}
]
properties {
 bOverrideConcurrency=true
 maxDistance=13000
 soundClassObject=/Game/Sounds/FortSoundClassesAndModes/Fort_Explosives
 volumeMultiplier=0.3
 attenuationOverrides {distanceAlgorithm=NaturalSound, falloffDistance=2000, stereoSpread=0}
 concurrencyOverrides {maxCount=3}
}
