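# Sound cue for the TeaCake dynamite landing, built from three TNT bounce recordings.
# A Random node picks one of the equally weighted waves and feeds a Modulator that slightly varies pitch.
# It plays at reduced volume in the loot-collect sound class with weapon-drop attenuation.
# It references a shared pickup concurrency set and carries single-instance, prevent-new concurrency overrides.
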SoundCue /TeaCakeConsumables/Assets/Audio/SC_TeaCake_Dynamite_Land
root=Modulator_0
nodes [
 {id=Modulator_0, children [Random_0], properties {pitchMax=1.1, pitchMin=0.9}}
 {id=Random_0, children [WavePlayer_0,WavePlayer_1,WavePlayer_2], properties {weights [1,1,1]}}
 {id=WavePlayer_0, properties {soundWaveAssetPtr=&/TNT_Explode_Bounce_01}}
 {id=WavePlayer_1, properties {soundWaveAssetPtr=&/TNT_Explode_Bounce_02}}
 {id=WavePlayer_2, properties {soundWaveAssetPtr=&/TNT_Explode_Bounce_03}}
]
properties {
 attenuationSettings=&Fort_Weapon_Drop_Att
 maxDistance=3000
 soundClassObject=&Fort_Loot_Collect
 volumeMultiplier=0.3
 concurrencyOverrides {maxCount=1, resolutionRule=PreventNew}
 concurrencySet [&PickupMaximum_Oldest]
}
