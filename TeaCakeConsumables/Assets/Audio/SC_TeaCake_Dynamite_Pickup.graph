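# Sound cue for picking up the TeaCake dynamite consumable, layering a pickup sound with a fuse ignition.
# A mixer blends the reused frag grenade pickup wave, given slight pitch variation, with one of two random TNT ignite waves pitched well up by a fixed modulator.
# It plays at reduced volume through the loot-collect sound class, with weapon-pickup attenuation and a shared pickup concurrency set.
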SoundCue /TeaCakeConsumables/Assets/Audio/SC_TeaCake_Dynamite_Pickup
root=Mixer_0
nodes [
 {id=Mixer_0, children [Modulator_1,Modulator_0], properties {inputVolume [1,1]}}
 {id=Modulator_0, children [Random_0], properties {pitchMax=1.7, pitchMin=1.7}}
 {id=Modulator_1, children [WavePlayer_5], properties {pitchMax=1.1, pitchMin=0.9, volumeMax=1, volumeMin=1}}
 {id=Random_0, children [WavePlayer_1,WavePlayer_0], properties {weights [1,1]}}
 {id=WavePlayer_0, properties {soundWaveAssetPtr=&/TNT_Ignite_04}}
 {id=WavePlayer_1, properties {soundWaveAssetPtr=&/TNT_Ignite_03}}
 {id=WavePlayer_5, properties {soundWaveAssetPtr=/Game/Sounds/Fort_Weapons/Grenades/Frag/Grenade_Frag_Pickup_01}}
]
properties {
 attenuationSettings=&Fort_Weapon_Pickup_Att
 maxDistance=3500
 soundClassObject=&Fort_Loot_Collect
 volumeMultiplier=0.5
 concurrencyOverrides {resolutionRule=PreventNew}
 concurrencySet [&PickupMaximum_Oldest]
}
